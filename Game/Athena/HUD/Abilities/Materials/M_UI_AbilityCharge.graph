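# Additive UI material for an ability charge meter, combining a charge texture with an edge flare.
# ChargePercent feeds a clamp lerp that drives the texture's vertical scale mask, the flare's vertical offset and a cutoff to black.
# A timed sine feeds a flicker lerp scaling texture brightness and the desaturated flare's size.
# As written, both the clamp and flicker lerps use red for both endpoints, so neither varies with its alpha.
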Material /Game/Athena/HUD/Abilities/Materials/M_UI_AbilityCharge
graph {
 properties {blendMode=BLEND_Additive, materialDomain=MD_UI, pixelDepthOffsetMode=PDOM_Legacy, usedShadingModels=Default Lit}
 nodes [
  n1 = ScalarParameter() {parameterName=CooldownPosition}
  n2 = VectorParameter() {parameterName=CooldownImageClamp, defaultValue=<rgba 0.1429 0.819 0 1>}
  n3 = LinearInterpolate(A=@n2.R, B=@n2.R, Alpha=ScalarParameter() {defaultValue=0.5, parameterName=ChargePercent, sliderMax=1})
  n4 = LinearInterpolate(Alpha=@n3) {constA=-0.5, constB=0.5}
  n5 = Add(A=TextureCoordinate(), B=AppendVector(A=Constant(), B=Add(A=@n1, B=@n4)))
  n6 = VectorParameter() {parameterName=FlickerExtents, defaultValue=<rgba 0.5 1 0 1>}
  n7 = Time()
  n8 = ScalarParameter() {defaultValue=1, parameterName=EdgeFlickerSpeed}
  n9 = LinearInterpolate(
   A=@n6.R
   B=@n6.R
   Alpha=MaterialFunctionCall(Value=Multiply(A=@n7, B=@n8)) {materialFunction=&LinearSine}.Linear Sine
  )
  n10 = VectorParameter() {parameterName=CooldownEdgeFlareScale, defaultValue=<rgba 1 1 0 1>}
  n11 = AppendVector(A=Multiply(A=@n9, B=@n10.R), B=Multiply(A=@n9, B=@n10.G))
  n12 = Multiply(A=TextureSample(UVs=MaterialFunctionCall(UVs=@n5, Texture Scale=@n11) {materialFunction=&ScaleUVsByCenter}.UVs) {samplerSource=SSM_Clamp_WorldGroupSettings, texture=/Game/Athena/HUD/Materials/Victory/T_UI_StandaloneFlare}.RGB)
  n13 = Constant() {r=1}
  n14 = Constant() {r=1}
  n15 = If(A=@n3, B=@n2.G, "A > B"=@n14, "A == B"=@n14, "A < B"=@n3)
  n16 = Multiply(
   A=TextureSampleParameter2D() {parameterName=Texture, texture=&/T_UI_Hightower_Charge_Secondary}.RGB
   B=MaterialFunctionCall!(Texture Scale=AppendVector(A=@n13, B=@n15), UVs=TextureCoordinate(), Offset XY Texturesize Z=Constant3Vector() {constant=<rgba 0 1 1 0>}) {materialFunction=/Engine/Functions/Engine_MaterialFunctions01/Texturing/ScaleUVsAroundPoint}.0-1 mask
  )
  n17 = Constant()
  If(A=@n3, B=@n2.R, "A > B"=Add(A=Desaturation(None=@n12), B=Multiply(A=@n9, B=@n16)), "A == B"=@n17, "A < B"=@n17)
  ScalarParameter() {defaultValue=0.5, parameterName=ChargePercent, sliderMax=1}
 ]
}
