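# Additive UI material for an ability cooldown effect in the Athena HUD, built from three radially masked glow layers that are summed.
# One layer pans a tinted fabric noise, another takes the brighter of two panning starburst samples, tinted and power-shaped.
# The third sums a standalone flare and two tinted fake lens flares with scale and offset controls.
# Both lens flares rotate by the same Flare2Angle parameter.
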Material /Game/Athena/HUD/Abilities/Materials/M_UI_AbilityCooldownEffect
graph {
 properties {blendMode=BLEND_Additive, materialDomain=MD_UI, numCustomizedUVs=1, pixelDepthOffsetMode=PDOM_Legacy, usedShadingModels=Default Lit}
 nodes [
  n1 = MaterialFunctionCall(
   UVs=MaterialFunctionCall(UVs=TextureCoordinate(), Texture Scale=VectorParameter() {group=Radial2Mask, parameterName=Radial2Scale, defaultValue=<rgba 1 1 0 0>}.RGB) {materialFunction=&ScaleUVsByCenter}.UVs
   CenterPosition=VectorParameter() {group=Radial2Mask, parameterName="CenterPosition2 (V2)", defaultValue=<rgba 0.5 0.5 0 0>}.RGB
   Radius=ScalarParameter() {defaultValue=0.5, group=Radial2Mask, parameterName="Radius2 (S)"}
   Density=ScalarParameter() {defaultValue=2, group=Radial2Mask, parameterName="Density2 (S)"}
  ) {materialFunction=&RadialGradientExponential}
  n2 = VectorParameter() {group=Pattern1, parameterName=PatternScale, defaultValue=<rgba 0.01 2 0 0>}
  n3 = Panner(
   Coordinate=Multiply(A=TextureCoordinate(), B=AppendVector(A=@n2.R, B=@n2.R))
   Speed=AppendVector(A=ScalarParameter() {defaultValue=1, group=Pattern1, parameterName=Pattern1Speed}, B=Constant())
  )
  n4 = Multiply(
   A=Power(
    Base=TextureSample(UVs=@n3) {samplerType=SAMPLERTYPE_LinearColor, texture=/Game/Global/Textures/Default/Noise/T_FN_FabricTile_B}.R
    Exp=ScalarParameter() {defaultValue=1, group=Pattern1, parameterName=Pattern1Exp}
   )
   B=VectorParameter() {group=Pattern1, parameterName=Pattern1Color, defaultValue=<rgba 0 0.7323 1 1>}.RGB
  )
  n5 = MaterialFunctionCall(
   UVs=MaterialFunctionCall(UVs=TextureCoordinate(), Texture Scale=VectorParameter() {group=RadialMask, parameterName=RadialScale, defaultValue=<rgba 1 1 0 0>}.RGB) {materialFunction=&ScaleUVsByCenter}.UVs
   CenterPosition=VectorParameter() {group=RadialMask, parameterName="CenterPosition (V2)", defaultValue=<rgba 0.5 0.5 0 0>}.RGB
   Radius=ScalarParameter() {defaultValue=0.5, group=RadialMask, parameterName="Radius (S)"}
   Density=ScalarParameter() {defaultValue=2, group=RadialMask, parameterName="Density (S)"}
  ) {materialFunction=&RadialGradientExponential}
  n6 = VectorParameter() {group=Flare1, parameterName=FlareOffset, defaultValue=<rgba 0 0 0 1>}
  n7 = Add(
   A=MaterialFunctionCall(UVs=TextureCoordinate(), Texture Scale=VectorParameter() {group=Flare1, parameterName=Flare Scale, defaultValue=<rgba 1 1 0 0>}.RGB) {materialFunction=&ScaleUVsByCenter}.UVs
   B=AppendVector(A=@n6.R, B=@n6.R)
  )
  n8 = MaterialFunctionCall(UVs=TextureCoordinate(), Texture Scale=VectorParameter() {group=Flare2, parameterName=Flare2 Scale, defaultValue=<rgba 1 1 0 0>}.RGB) {materialFunction=&ScaleUVsByCenter}
  n9 = ScalarParameter() {group=Flare2, parameterName=Flare2Angle}
  n10 = VectorParameter() {group=Flare2, parameterName=Flare2Offset, defaultValue=<rgba 0 0 0 1>}
  n11 = TextureSample(UVs=Add(
    A=MaterialFunctionCall(UVs=@n8.UVs, "Rotation Angle (0-1)"=@n9) {materialFunction=&CustomRotator}.Rotated Values
    B=AppendVector(A=@n10.R, B=@n10.R)
   )) {samplerSource=SSM_Clamp_WorldGroupSettings, texture=/Game/Athena/HUD/Materials/T_FakeLensFlare}
  n12 = VectorParameter() {group=Flare2, parameterName=Flare2Color, defaultValue=<rgba 0 0.4438 1 1>}
  n13 = TextureCoordinate()
  n14 = VectorParameter() {group=Flare3, parameterName=Flare3 Scale, defaultValue=<rgba 1 1 0 0>}
  n15 = VectorParameter() {group=Flare3, parameterName=Flare3Offset, defaultValue=<rgba 0 0 0 1>}
  n16 = Add(
   A=MaterialFunctionCall(
    UVs=MaterialFunctionCall(UVs=@n13, Texture Scale=@n14.RGB) {materialFunction=&ScaleUVsByCenter}.UVs
    "Rotation Angle (0-1)"=ScalarParameter() {group=Flare2, parameterName=Flare2Angle}
   ) {materialFunction=&CustomRotator}.Rotated Values
   B=AppendVector(A=@n15.R, B=@n15.R)
  )
  n17 = Add(
   A=Add(
    A=TextureSample(UVs=@n7) {samplerSource=SSM_Clamp_WorldGroupSettings, texture=/Game/Athena/HUD/Materials/Victory/T_UI_StandaloneFlare}.RGB
    B=Multiply(A=@n11.RGB, B=@n12.RGB)
   )
   B=Multiply(
    A=TextureSample(UVs=@n16) {texture=&T_FakeLensFlare}.RGB
    B=VectorParameter() {group=Flare3, parameterName=Flare3Color, defaultValue=<rgba 0 0.4438 1 1>}.RGB
   )
  )
  n18 = MaterialFunctionCall(UVs=TextureCoordinate(), Texture Scale=VectorParameter() {group=Radial3Mask, parameterName=Radial3Scale, defaultValue=<rgba 1 1 0 0>}.RGB) {materialFunction=&ScaleUVsByCenter}
  n19 = VectorParameter() {group=Radial3Mask, parameterName="CenterPosition3 (V2)", defaultValue=<rgba 0.5 0.5 0 0>}
  n20 = ScalarParameter() {defaultValue=0.5, group=Radial3Mask, parameterName="Radius3 (S)"}
  n21 = ScalarParameter() {defaultValue=2, group=Radial3Mask, parameterName="Density3 (S)"}
  n22 = VectorParameter() {group=Pattern2, parameterName=Pattern2Scale, defaultValue=<rgba 0.01 2 0 0>}
  n23 = Panner(
   Coordinate=Multiply(A=TextureCoordinate(), B=AppendVector(A=@n22.R, B=@n22.R))
   Speed=AppendVector(A=ScalarParameter() {defaultValue=1, group=Pattern2, parameterName=Pattern2Speed}, B=Constant())
  )
  n24 = VectorParameter() {group=Pattern2, parameterName=Pattern3Scale, defaultValue=<rgba 0.01 2 0 0>}
  n25 = Panner(
   Coordinate=Multiply(A=TextureCoordinate(), B=AppendVector(A=@n24.R, B=@n24.R))
   Speed=AppendVector(A=ScalarParameter() {defaultValue=1, group=Pattern2, parameterName=Pattern3Speed}, B=Constant())
  )
  n26 = Multiply(
   A=Max(
    A=TextureSample(UVs=@n23) {texture=/Game/Athena/UI/InGame/Interrogation/Materials/T_UI_StarBurstPattern}.R
    B=TextureSample(UVs=@n25) {texture=&T_UI_StarBurstPattern}.R
   )
   B=VectorParameter() {group=Pattern2, parameterName=Pattern2Color, defaultValue=<rgba 0 0.7323 1 1>}.RGB
  )
  n27 = ScalarParameter() {defaultValue=1, group=Pattern2, parameterName=Pattern2Exp}
  Add(
   A=Add(A=Multiply(A=@n1.RadialGradientExponential, B=@n4), B=Multiply(A=@n5.RadialGradientExponential, B=@n17))
   B=Multiply(
    A=MaterialFunctionCall(UVs=@n18.UVs, CenterPosition=@n19.RGB, Radius=@n20, Density=@n21) {materialFunction=&RadialGradientExponential}.RadialGradientExponential
    B=Power(Base=@n26, Exp=@n27)
   )
  )
  MaterialFunctionCall(UVs=TextureCoordinate(), "Rotation Angle (0-1)"=Constant() {r=0.25}) {materialFunction=&CustomRotator}
 ]
}
